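# Physics asset for the IndieBucket backpack cosmetic, giving collision bodies and joints to an eared figure with arms, legs and headphones.
# Capsules, spheres and boxes form a tree from the pack through a tightly limited Body joint to the head, two-segment ears and arms, and single-body legs.
# Arm joints swing and twist freely, while head, ear and leg joints are limited.
# The two-segment headphone chain hangs off the upper left arm with limited joints.
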PhysicsAsset /BRCosmetics/Accessories/FORT_Backpacks/Backpack_F_MED_IndieBucket/Meshes/F_MED_IndieBucket_Pack_Physics
bodies {
 Arm_L_01 [{shapeType=Capsule, length=2.6299, radius=1.4222, shapeName=Arm_L_01_capsule, center=<2.6594 0 0>, rotation=<rot 90 0 0>}]
 Arm_L_02 [{shapeType=Capsule, length=1.0399, radius=1.8113, shapeName=Arm_L_02_capsule, center=<2.2261 0 0>, rotation=<rot 90 0 0>}]
 Arm_R_01 [{shapeType=Capsule, length=2.6299, radius=1.4222, shapeName=Arm_L_01_capsule, center=<2.6594 0 0>, rotation=<rot 90 0 0>}]
 Arm_R_02 [{shapeType=Capsule, length=1.0399, radius=1.8113, shapeName=Arm_L_02_capsule, center=<2.2261 0 0>, rotation=<rot 90 0 0>}]
 Body [{shapeType=Sphere, radius=5.0174, center=<0 0 -2>}]
 Ear_L_01 [{
   shapeType=Box
   extentX=1.9224
   extentY=2.2623
   extentZ=5.1797
   shapeName=Ear_L_01_box
   center=<0.1966 0.4415 -0.1211>
   rotation=<rot 0 173.8336 -73.9493>
  }]
 Ear_L_02 [{
   shapeType=Box
   extentX=6.8548
   extentY=3.6724
   extentZ=5.7046
   shapeName=Ear_L_02_box
   center=<1.9245 0.0314 -1.1007>
   rotation=<rot 0 159.7423 -88.1336>
  }]
 Ear_R_01 [{
   shapeType=Box
   extentX=2.0106
   extentY=2.171
   extentZ=5.1797
   shapeName=Ear_R_01_box
   center=<-0.0096 -0.4185 0.0269>
   rotation=<rot 0 8.4098 -77.6979>
  }]
 Ear_R_02 [{
   shapeType=Box
   extentX=6.6971
   extentY=3.3897
   extentZ=5.705
   shapeName=Ear_R_02_box
   center=<2.0111 0.3255 -0.6926>
   rotation=<rot 0 23.7074 -88.2277>
  }]
 Head [{shapeType=Capsule, length=1, radius=1, shapeName=dyn_head_capsule, center=<0 0 2>, rotation=<rot 0 0 -8>}]
 HeadPhone_01 [{
   shapeType=Capsule
   length=9.7049
   radius=1.5754
   shapeName=dyn_cord_01_capsule
   center=<-0.1158 0.2425 -0.8623>
   rotation=<rot 62.4607 -36.8323 56.4023>
  }]
 HeadPhone_02 [{
   shapeType=Capsule
   length=9.7049
   radius=1.5754
   shapeName=dyn_cord_01_capsule
   center=<-0.3421 0.4119 -1.4043>
   rotation=<rot 62.4607 -36.8323 56.4023>
  }]
 Leg_L_01 [{shapeType=Sphere, radius=3.2226, center=<1.1929 -0.0238 -0.1217>}]
 Leg_R_01 [{shapeType=Sphere, radius=3.2226, center=<1.1929 -0.0238 -0.1217>}]
 pack [{shapeType=Capsule, length=1, radius=1}]
}
constraints [
 {bone1Name=Body, bone2Name=pack, swing1LimitDegrees=1, swing2LimitDegrees=1, twistLimitDegrees=1}
 {bone1Name=Head, bone2Name=Body, swing1LimitDegrees=5, swing2LimitDegrees=10, twistLimitDegrees=0}
 {bone1Name=Leg_L_01, bone2Name=Body, swing1LimitDegrees=15, swing2LimitDegrees=30, twistLimitDegrees=20}
 {bone1Name=Leg_R_01, bone2Name=Body, swing1LimitDegrees=15, swing2LimitDegrees=30, twistLimitDegrees=20}
 {bone1Name=Ear_R_02, bone2Name=Ear_R_01, swing1LimitDegrees=5, swing2LimitDegrees=25, twistLimitDegrees=2}
 {bone1Name=Ear_R_01, bone2Name=Head, swing1LimitDegrees=5, swing2LimitDegrees=15, twistLimitDegrees=2}
 {bone1Name=Ear_L_02, bone2Name=Ear_L_01, swing1LimitDegrees=5, swing2LimitDegrees=25, twistLimitDegrees=2}
 {bone1Name=Ear_L_01, bone2Name=Head, swing1LimitDegrees=5, swing2LimitDegrees=15, twistLimitDegrees=2}
 {bone1Name=HeadPhone_01, bone2Name=Arm_L_01, swing1LimitDegrees=25, swing2LimitDegrees=40, twistLimitDegrees=3}
 {
  bone1Name=Arm_L_02
  bone2Name=Arm_L_01
  swing1LimitDegrees=45
  swing1Motion=Free
  swing2LimitDegrees=45
  swing2Motion=Free
  twistLimitDegrees=45
  twistMotion=Free
 }
 {
  bone1Name=Arm_L_01
  bone2Name=Body
  swing1LimitDegrees=45
  swing1Motion=Free
  swing2LimitDegrees=45
  swing2Motion=Free
  twistLimitDegrees=45
  twistMotion=Free
 }
 {
  bone1Name=Arm_R_02
  bone2Name=Arm_R_01
  swing1LimitDegrees=45
  swing1Motion=Free
  swing2LimitDegrees=45
  swing2Motion=Free
  twistLimitDegrees=45
  twistMotion=Free
 }
 {
  bone1Name=Arm_R_01
  bone2Name=Body
  swing1LimitDegrees=45
  swing1Motion=Free
  swing2LimitDegrees=45
  swing2Motion=Free
  twistLimitDegrees=45
  twistMotion=Free
 }
 {bone1Name=HeadPhone_02, bone2Name=HeadPhone_01, swing1LimitDegrees=25, swing2LimitDegrees=40, twistLimitDegrees=3}
]
