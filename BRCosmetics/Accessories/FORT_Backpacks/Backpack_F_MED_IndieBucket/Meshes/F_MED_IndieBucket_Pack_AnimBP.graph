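# Animation Blueprint for the IndieBucket backpack cosmetic, adding procedural secondary motion to its arms, head, headphones and ears.
# A chain of AnimDynamics nodes feeds additive arm rotations gated by Jumping (default true, never set here), then a RigidBody simulation.
# An additive root offset follows Aim Down Sight, set from bIsTargeting each update with interpolated alpha.
# The spine-based rotation calculation is left unconnected, and backpack_rotate is only zeroed.
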AnimBlueprint /BRCosmetics/Accessories/FORT_Backpacks/Backpack_F_MED_IndieBucket/Meshes/F_MED_IndieBucket_Pack_AnimBP
target_skeleton=&/F_MED_IndieBucket_Pack_Skeleton
graphs [
 AnimGraph {nodes [
   n1 = Anim Dynamics - Bone: Arm_R_02(ComponentPose=Local To Component(LocalPose=Input Pose()), Alpha=0.5, ExternalForce="0,0,0") {
    angularDampingOverride=1
    angularSpringConstant=4000
    bAngularSpring=true
    bGravityOverrideInSimSpace=true
    bOverrideAngularDamping=true
    gravityScale=0.35
    simulationSpace=RootRelative
    alphaBoolBlend {blendOption=Linear}
    alphaScaleBiasClamp {scale=0.8}
    boundBone {boneName=Arm_R_02}
    componentLinearAccScale=<0.8 0.8 0.8>
    componentLinearVelScale=<0.8 0.8 0.8>
    physicsBodyDefinitions [{
      sphereCollisionRadius=0
      boundBone {boneName=Arm_R_02}
      boxExtents=<10 10 10>
      constraintSetup {angularLimitsMax=<5 1 2>, angularLimitsMin=<-5 -5 -2>, angularTarget=<1 0 0>}
      localJointOffset=<1 0 0>
     }]
   }
   n2 = Anim Dynamics - Bone: Arm_L_02(ComponentPose=@n1, Alpha=0.5, ExternalForce="0,0,0") {
    angularDampingOverride=1
    angularSpringConstant=4000
    bAngularSpring=true
    bGravityOverrideInSimSpace=true
    bOverrideAngularDamping=true
    gravityScale=0.35
    simulationSpace=RootRelative
    alphaBoolBlend {blendOption=Linear}
    alphaScaleBiasClamp {scale=0.8}
    boundBone {boneName=Arm_L_02}
    componentLinearAccScale=<0.8 0.8 0.8>
    componentLinearVelScale=<0.8 0.8 0.8>
    physicsBodyDefinitions [{
      sphereCollisionRadius=0
      boundBone {boneName=Arm_L_02}
      boxExtents=<10 10 10>
      constraintSetup {angularLimitsMax=<5 1 2>, angularLimitsMin=<-5 -5 -2>, angularTarget=<1 0 0>}
      localJointOffset=<1 0 0>
     }]
   }
   n3 = Anim Dynamics - Bone: Arm_R_01(ComponentPose=@n2, Alpha=0.5, ExternalForce="0,0,0") {
    angularDampingOverride=1
    angularSpringConstant=4000
    bAngularSpring=true
    bGravityOverrideInSimSpace=true
    bOverrideAngularDamping=true
    gravityScale=0.35
    simulationSpace=RootRelative
    alphaBoolBlend {blendOption=Linear}
    alphaScaleBiasClamp {scale=0.8}
    boundBone {boneName=Arm_R_01}
    componentLinearAccScale=<0.8 0.8 0.8>
    componentLinearVelScale=<0.8 0.8 0.8>
    physicsBodyDefinitions [{
      sphereCollisionRadius=0
      boundBone {boneName=Arm_R_01}
      boxExtents=<10 10 10>
      constraintSetup {angularLimitsMax=<15 2 5>, angularLimitsMin=<-20 -20 -5>, angularTarget=<1 0 0>}
      localJointOffset=<1 0 0>
     }]
   }
   n4 = Anim Dynamics - Bone: Arm_L_01(ComponentPose=@n3, Alpha=0.5, ExternalForce="0,0,0") {
    angularDampingOverride=1
    angularSpringConstant=4000
    bAngularSpring=true
    bGravityOverrideInSimSpace=true
    bOverrideAngularDamping=true
    gravityScale=0.35
    simulationSpace=RootRelative
    alphaBoolBlend {blendOption=Linear}
    alphaScaleBiasClamp {scale=0.8}
    boundBone {boneName=Arm_L_01}
    componentLinearAccScale=<0.8 0.8 0.8>
    componentLinearVelScale=<0.8 0.8 0.8>
    physicsBodyDefinitions [{
      sphereCollisionRadius=0
      boundBone {boneName=Arm_L_01}
      boxExtents=<10 10 10>
      constraintSetup {angularLimitsMax=<15 2 5>, angularLimitsMin=<-20 -20 -5>, angularTarget=<1 0 0>}
      localJointOffset=<1 0 0>
     }]
   }
   n5 = Anim Dynamics - Bone: Head(ComponentPose=@n4, Alpha=0.5, ExternalForce="0,0,0") {
    angularDampingOverride=1
    angularSpringConstant=4000
    bAngularSpring=true
    bGravityOverrideInSimSpace=true
    bOverrideAngularDamping=true
    gravityScale=0.35
    simulationSpace=RootRelative
    alphaBoolBlend {blendOption=Linear}
    alphaScaleBiasClamp {scale=0.8}
    boundBone {boneName=Head}
    componentLinearAccScale=<0.8 0.8 0.8>
    componentLinearVelScale=<0.8 0.8 0.8>
    physicsBodyDefinitions [{
      sphereCollisionRadius=0
      boundBone {boneName=Head}
      boxExtents=<20 20 20>
      constraintSetup {angularLimitsMax=<50 0 0>, angularLimitsMin=<-30 -8 -20>, angularTarget=<1 0 0>}
      localJointOffset=<1 0 0>
     }]
   }
   n6 = Anim Dynamics - Bone: HeadPhone_01(ComponentPose=@n5, Alpha=0.7, ExternalForce="0,0,0") {
    angularDampingOverride=0.99
    angularSpringConstant=4000
    bOverrideAngularDamping=true
    bUseGravityOverride=true
    gravityScale=0.35
    simulationSpace=RootRelative
    alphaBoolBlend {blendOption=Linear}
    boundBone {boneName=HeadPhone_01}
    componentAppliedLinearAccClamp=<5000 5000 5000>
    componentLinearAccScale=<0.5 0.5 0.5>
    gravityOverride {z=-980}
    physicsBodyDefinitions [{
      sphereCollisionRadius=0
      boundBone {boneName=HeadPhone_01}
      boxExtents=<5 10 10>
      constraintSetup {angularLimitsMax=<5 70 5>, angularLimitsMin=<-5 -100 -5>, angularTarget=<1 0 0>}
      localJointOffset=<3 0 0>
     }]
    planarLimits [{drivingBone {boneName=HeadPhone_01}, planeTransform {location=<-8 0 0>, rotation=<rot 0 0 0>}}]
   }
   n7 = Anim Dynamics - Bone: HeadPhone_02(ComponentPose=@n6, Alpha=0.45, ExternalForce="0,0,0") {
    angularDampingOverride=0.99
    angularSpringConstant=1000
    bOverrideAngularDamping=true
    bUseGravityOverride=true
    bUsePlanarLimit=false
    gravityScale=0.35
    simulationSpace=RootRelative
    alphaBoolBlend {blendOption=Linear}
    boundBone {boneName=HeadPhone_02}
    componentAppliedLinearAccClamp=<5000 5000 5000>
    componentLinearAccScale=<0.5 0.5 0.5>
    gravityOverride {z=-980}
    physicsBodyDefinitions [{
      sphereCollisionRadius=0
      boundBone {boneName=HeadPhone_02}
      boxExtents=<5 10 10>
      constraintSetup {angularLimitsMax=<2 70 5>, angularLimitsMin=<-2 -60 -5>, angularTarget=<1 0 0>}
      localJointOffset=<3 0 0>
     }]
   }
   n8 = Anim Dynamics - Bone: Ear_R_01(ComponentPose=@n7, Alpha=1, ExternalForce="0,0,0") {
    angularDampingOverride=0.99
    angularSpringConstant=5000
    bOverrideAngularDamping=true
    bUseGravityOverride=true
    simulationSpace=BoneRelative
    alphaBoolBlend {blendOption=Linear}
    boundBone {boneName=Ear_R_01}
    componentAppliedLinearAccClamp=<3000 3000 3000>
    componentLinearAccScale=<0.2 0.2 0.2>
    componentLinearVelScale=<0.3 0.3 0.3>
    gravityOverride {z=-600}
    physicsBodyDefinitions [{
      boundBone {boneName=Ear_R_01}
      boxExtents=<10 20 20>
      constraintSetup {angularLimitsMax=<0.1 20 2>, angularLimitsMin=<-0.1 -15 -2>, angularTarget=<0 0 0>}
      localJointOffset=<-3 0 0.15>
     }]
   }
   n9 = Anim Dynamics - Bone: Ear_R_02(ComponentPose=@n8, Alpha=1, ExternalForce="0,0,0") {
    angularDampingOverride=0.99
    angularSpringConstant=2000
    bAngularSpring=true
    bGravityOverrideInSimSpace=true
    bOverrideAngularDamping=true
    bOverrideLinearDamping=true
    bUseGravityOverride=true
    gravityScale=0.35
    linearDampingOverride=0.99
    simulationSpace=BoneRelative
    alphaBoolBlend {blendOption=Linear}
    alphaScaleBiasClamp {scale=0.8}
    boundBone {boneName=Ear_R_02}
    componentAppliedLinearAccClamp=<4000 4000 4000>
    componentLinearAccScale=<0.6 0.6 0.6>
    componentLinearVelScale=<0.6 0.6 0.6>
    gravityOverride {z=-500}
    physicsBodyDefinitions [{
      sphereCollisionRadius=0
      boundBone {boneName=Ear_R_02}
      boxExtents=<10 10 10>
      constraintSetup {angularLimitsMax=<0 35 0>, angularLimitsMin=<0 -15 0>, angularTarget=<1 0 0>}
      localJointOffset=<1 0 0>
     }]
   }
   n10 = Anim Dynamics - Bone: Ear_L_01(ComponentPose=@n9, Alpha=0.8, ExternalForce="0,0,0") {
    angularDampingOverride=0.99
    angularSpringConstant=5000
    bOverrideAngularDamping=true
    bOverrideLinearDamping=true
    bUseGravityOverride=true
    linearDampingOverride=0.8
    simulationSpace=BoneRelative
    alphaBoolBlend {blendOption=Linear}
    boundBone {boneName=Ear_L_01}
    componentAppliedLinearAccClamp=<3000 3000 3000>
    componentLinearAccScale=<0.25 0.25 0.25>
    componentLinearVelScale=<0.35 0.35 0.35>
    gravityOverride {z=-600}
    physicsBodyDefinitions [{
      boundBone {boneName=Ear_L_01}
      boxExtents=<10 20 20>
      constraintSetup {angularLimitsMax=<0.1 10 2>, angularLimitsMin=<-0.1 -12 -2>, angularTarget=<0 0 0>}
      localJointOffset=<-3 0 0.15>
     }]
   }
   n11 = Anim Dynamics - Bone: Ear_L_02(ComponentPose=@n10, Alpha=0.8, ExternalForce="0,0,0") {
    angularDampingOverride=0.99
    angularSpringConstant=2000
    bAngularSpring=true
    bGravityOverrideInSimSpace=true
    bOverrideAngularDamping=true
    bOverrideLinearDamping=true
    bUseGravityOverride=true
    gravityScale=0.35
    linearDampingOverride=0.8
    simulationSpace=BoneRelative
    alphaBoolBlend {blendOption=Linear}
    alphaScaleBiasClamp {scale=0.8}
    boundBone {boneName=Ear_L_02}
    componentAppliedLinearAccClamp=<5000 5000 5000>
    componentLinearAccScale=<0.35 0.35 0.35>
    componentLinearVelScale=<0.25 0.25 0.25>
    gravityOverride {z=-500}
    physicsBodyDefinitions [{
      sphereCollisionRadius=0
      boundBone {boneName=Ear_L_02}
      boxExtents=<10 10 10>
      constraintSetup {angularLimitsMax=<0 25 0>, angularLimitsMin=<0 -40 0>, angularTarget=<1 0 0>}
      localJointOffset=<1 0 0>
     }]
   }
   n12 = Get Jumping() -> (Jumping:bool)
   n13 = "Transform (Modify) Bone - Bone: Arm_L_02"!(ComponentPose=@n11, bAlphaBoolEnabled:bool=@n12, Translation="0,0,0", Rotation="-11,0,0", Scale="1,1,1") -> (Pose) {
    alphaInputType=Bool
    rotationMode=BMM_Additive
    rotationSpace=BCS_BoneSpace
    alphaBoolBlend {blendOption=Linear}
    boneToModify {boneName=Arm_L_02}
    rotation {pitch=-11}
   }
   n14 = "Transform (Modify) Bone - Bone: Arm_R_02"!(ComponentPose=@n13, bAlphaBoolEnabled:bool=@n12, Translation="0,0,0", Rotation="-11,0,0", Scale="1,1,1") -> (Pose) {
    alphaInputType=Bool
    rotationMode=BMM_Additive
    rotationSpace=BCS_BoneSpace
    alphaBoolBlend {blendOption=Linear}
    boneToModify {boneName=Arm_R_02}
    rotation {pitch=-11}
   }
   n15 = RigidBody(ComponentPose=@n14) {
    bOverrideWorldGravity=true
    overridePhysicsAsset=&/F_MED_IndieBucket_Pack_Physics
    simulationSpace=BaseBoneSpace
    alphaBoolBlend {blendOption=Linear}
    baseBoneRef {boneName=root}
    componentAppliedLinearAccClamp=<1000 1000 1000>
    componentLinearAccScale=<0.6 0.6 0.6>
    componentLinearVelScale=<0.4 0.4 0.4>
    overrideWorldGravity {z=-980}
   }
   n16 = "Transform (Modify) Bone - Bone: root"!(ComponentPose=@n15, Alpha=Get Aim Down Sight() -> (Aim Down Sight:double), Translation="0,1,3", Rotation="0,0,13.5") -> (Pose) {
    rotationMode=BMM_Additive
    rotationSpace=BCS_BoneSpace
    translationMode=BMM_Additive
    translationSpace=BCS_BoneSpace
    alphaBoolBlend {blendOption=Linear}
    alphaScaleBiasClamp {bInterpResult=true}
    boneToModify {boneName=root}
    rotation {roll=13.5}
    translation {y=1, z=3}
   }
   Output Pose(Result=Component To Local(ComponentPose=@n16)) {layerGroup=None}
  ]}
 EventGraph {nodes [
   Event BlueprintUpdateAnimation()
   n1 = Get bIsTargeting()
   > Set Aim Down Sight(Aim Down Sight:double="To Float (Boolean)"(InBool=@n1)) -> (Output_Get:double)
   > Set backpack_rotate(backpack_rotate="0, 0, 0")
   TryGetPawnOwner()
   n2 = BreakTransform(InTransform=GetDeltaTransformFromRefPose(self=GetOwnerSkeletalMeshComponent(), BoneName=spine_05, BaseName=spine_01))
   MakeRotator(Yaw=MapRangeClamped(Value=BreakRotator(InRot=@n2.Rotation).Yaw, InRangeA=-5, InRangeB=-30, OutRangeB=15))
   TryGetPawnOwner()
   Set Is Targeting(Is Targeting:bool=@n1) -> (Output_Get:bool)
  ]}
]
variables [
 Aim Down Sight {type=number}
 backpack_rotate {type=Rotator}
 Is Targeting {type=boolean}
 Jumping {type=boolean, default=true}
]
